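; Parent Blueprint for Gnice Gnomes quest props: gnomes with iced feet that giggle as a hint until collected.
; On quest init, curve-table rows supply hotfixable hint sound values and a looping timer plays the giggle.
; Completing the objective stops the timer, breaks the ice with sound and particles, then timeline-shrinks and hides the gnome.
; Quest updates disable collision, stop hints and hide the prop if completed-objective count is at least zero.
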(event EventQuestInitializedAndValid
  ;[1 1 1] Initialize hotfixable hint sound vars
  (bind _outxy (Fortnite|EvaluateCurveTableRow (Variables|Default|GetHintSoundVolRow)))
  (Variables|Default|SetHintSoundVolumeLevel _outxy)
  (bind _outxy_2 (Fortnite|EvaluateCurveTableRow (Variables|Default|GetHintSoundReplayRow)))
  (Variables|Default|SetHintSoundReplayTime _outxy_2)
  ;[1 1 1] Set timer for looping hint sound
  (bind _returnvalue (Utilities|Time|SetTimerbyFunctionName 0 "PlayGnomeGiggleSound" (Variables|Default|GetHintSoundReplayTime) true))
  (Variables|Default|SetHintSoundTimer _returnvalue))

(event EventThisQuestObjectiveComplete
  (params (IneractingPawn :type Fort%20Pawn%20Object%20Reference))
  (bind _returnvalue (Transformation|GetActorLocation))
  ;[1 1 1] Turn off Hint sound timer and play Sigh
  (Utilities|Time|ClearandInvalidateTimerbyHandle (Variables|Default|GetHintSoundTimer))
  (Audio|PlaySoundatLocation (Variables|Default|GetIceBreakSound) _returnvalue)
  (Development|SetHiddeninGame (Variables|Default|GetSM_Ice_Feet_Cube) true false)
  (Components|Activation|Activate (Variables|Default|GetP_Athena_Ice_Foot_Break) false)
  (Utilities|FlowControl|Delay 0.2)
  (Audio|PlaySoundatLocation (Variables|Default|GetCollectedSound) _returnvalue)
  (bind (_direction _shrinktime) (Timeline|ShrinkTime PlayFromStart)
    (:Update
      (Transformation|SetRelativeScale3D (Variables|Default|GetProp_Lawn_Ornament_Gnome_02) (Math|Vector|Lerp(Vector) "1.000000,1.000000,1.000000" "0, 0, 0" _shrinktime)))
    (:Finished
      (Rendering|SetActorHiddenInGame true))))

(event EventQuestUpdatedFromAnywhere
  ;[1 1 1] Check if Quest Complete after collection, if so hide the rest
  (|CallParentFunction)
  (bind _returnvalue (Quest|GetNumCompletedObjectives (Quests|GetQuestManager (Utilities|Array|Get(acopy) (Fortnite|GetLocalFortPlayerControllers true)) "Invalid") (Variables|QuestStuff|GetQuestItem)))
  (if (>= _returnvalue 0)
    (Collision|SetActorEnableCollision false)
    (Utilities|Time|ClearandInvalidateTimerbyHandle (Variables|Default|GetHintSoundTimer))
    (Utilities|FlowControl|Delay 0.2)
    (CallFunction|HideQuestProp)))

(fn PlayGnomeGiggleSound
  (Audio|PlaySoundatLocation (Variables|Default|GetHintSound) (Transformation|GetActorLocation)))

(fn ConstructionScript
  (|CallParentFunction)
  (Rendering|Components|Light|SetIntensity (Variables|Default|GetAmbientLight) 0.0))
